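# Physics asset for the Textile Ram backpack cosmetic, giving its handle and zipper pulls collision and limited physical motion.
# Two boxes form the pack body, and capsules cover the handle, three zipper bodies (left, front-right, back-right) and their hanging pulls.
# The handle and zippers are constrained to the pack, and each hanging pull chains to its zipper.
# Twist is nearly locked on every joint, and each zipper swings mainly on one axis while the pulls swing more freely.
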PhysicsAsset /BRCosmetics/Accessories/FORT_Backpacks/Backpack_F_MED_Textile_Ram/Meshes/F_MED_Textile_Ram_Pack_Physics
bodies {
 dyn_handle [{
   shapeType=Capsule
   length=3.4508
   radius=2.3287
   shapeName=dyn_handle_capsule
   center=<0.0652 -0.0629 2.4528>
   rotation=<rot 0 -89.8688 -87.6804>
  }]
 dyn_hang_b_r [{
   shapeType=Capsule
   length=3.1523
   radius=0.6349
   shapeName=dyn_hang_b_r_capsule
   center=<2.5177 0.0438 0.1399>
   rotation=<rot 0 -90.0076 -86.1767>
  }]
 dyn_hang_f_r [{
   shapeType=Capsule
   length=3.6768
   radius=0.6459
   shapeName=dyn_hang_f_r_capsule
   center=<2.1482 -0.153 -0.1011>
   rotation=<rot 0 -90.0557 -87.0542>
  }]
 dyn_hang_l [{
   shapeType=Capsule
   length=4
   radius=0.6
   shapeName=dyn_hang_l_capsule
   center=<2.2361 0.1753 0.0526>
   rotation=<rot -75.8843 86.954 -82.0064>
  }]
 dyn_zipper_b_r_MSTR [{
   shapeType=Capsule
   length=1.1284
   radius=0.5
   shapeName=dyn_zipper_b_r_MSTR_capsule
   center=<1.4709 0.0338 0.1815>
   rotation=<rot 0 89.4047 -81.8889>
  }]
 dyn_zipper_f_r_MSTR [{
   shapeType=Capsule
   length=0.6738
   radius=0.7055
   shapeName=dyn_zipper_f_r_MSTR_capsule
   center=<1.0999 0.1103 0.3647>
   rotation=<rot 0 92.7076 -74.7956>
  }]
 dyn_zipper_l_MSTR [{
   shapeType=Capsule
   length=0.95
   radius=0.5
   shapeName=dyn_zipper_l_MSTR_capsule
   center=<0.7249 0.0207 0.093>
   rotation=<rot -14.8014 -101.7463 -88.8148>
  }]
 pack [
  {
   shapeType=Box
   extentX=7.5799
   extentY=1.8534
   extentZ=13.6102
   shapeName=pack_box
   center=<10.3224 -6.9102 -9.7823>
   rotation=<rot 3.5662 -86.2265 -11.6624>
  }
  {
   shapeType=Box
   extentX=10.5
   extentY=1.5
   extentZ=11.5
   shapeName=pack_box
   center=<-9.327 -17.8365 -16.7323>
   rotation=<rot -2.802 -31.6265 14.887>
  }
 ]
}
constraints [
 {bone1Name=dyn_handle, bone2Name=pack, swing1LimitDegrees=30, swing2LimitDegrees=10, twistLimitDegrees=2}
 {bone1Name=dyn_zipper_l_MSTR, bone2Name=pack, swing1LimitDegrees=2, swing2LimitDegrees=48, twistLimitDegrees=1}
 {bone1Name=dyn_hang_l, bone2Name=dyn_zipper_l_MSTR, swing1LimitDegrees=45, swing2LimitDegrees=60, twistLimitDegrees=1}
 {bone1Name=dyn_zipper_f_r_MSTR, bone2Name=pack, swing1LimitDegrees=33.2857, swing2LimitDegrees=1.5, twistLimitDegrees=2}
 {bone1Name=dyn_hang_f_r, bone2Name=dyn_zipper_f_r_MSTR, swing1LimitDegrees=35, swing2LimitDegrees=35, twistLimitDegrees=1}
 {bone1Name=dyn_zipper_b_r_MSTR, bone2Name=pack, swing1LimitDegrees=28, swing2LimitDegrees=1.35, twistLimitDegrees=1}
 {bone1Name=dyn_hang_b_r, bone2Name=dyn_zipper_b_r_MSTR, swing1LimitDegrees=35, swing2LimitDegrees=35, twistLimitDegrees=1}
]
